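# Idle Niagara effect for the CanyonVale pickaxe, layering socket-attached fire and glows that respond to a reactivity level.
# FirePiece streams flame sprites, SignFire places six fire meshes, and SpriteGlows and TextureGlow add persistent glow sprites.
# System.ReactivityIntensity drives spawn rate, size and alpha, while User.Variant selects one of four color tints.
# SignFire also stores a smoothed position for lag, with time-of-day and front-end modifiers adjusting brightness.
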NiagaraSystem /BRCosmetics/Weapons/FORT_Melee/Pickaxe_CanyonVale/FX/NS_Pickaxe_CanyonVale_Idle
dependencies {
 usedDataInterfaces [&NiagaraDataInterfaceCurve, &NiagaraDataInterfaceColorCurve, &NiagaraDataInterfaceSocketReader]
 usedDynamicInputs [
  &MakeCustomFloatFromBool
  &Lerp_Float
  &Divide_Float
  &Subtract_Float
 ]
 usedModules [
  &self:SystemSpawnScript.SystemScriptSource.SystemScriptGraph.NiagaraNodeAssignment_0.SetVariables_#1
  &SystemState
  &self:SystemSpawnScript.SystemScriptSource.SystemScriptGraph.NiagaraNodeAssignment_3.SetVariables_#2
 ]
}
emitters {
 FirePiece {
  emitterUpdateScript [
   &EmitterState(Inactive Response=::NewEnumerator1, Life Cycle Mode=::NewEnumerator1, Loop Duration=1, UseLoopDelay=false)
   &SpawnRate(
    SpawnRate=&Multiply_Float(
     A=33
     B=&FloatFromCurve(
      FloatCurve=DataInterface() {propertyValues {Curve {keys [<at 0.2 0>,<at 0.8923 1>,<at 0.9367 1>,<at 1 1>]}}}
      CurveIndex=$System.ReactivityIntensity
     )
    )
    Use Spawn Probability=false
   )
  ]
  particleSpawnScript [
   &InitializeParticle(
    Color=<rgba 4 4 4 0.4522>
    Lifetime Max=0.7
    Lifetime Min=0.6
    Lifetime Mode=::NewEnumerator1
    Position Offset=<8 0 70>
    Sprite Rotation Mode=::NewEnumerator2
    Sprite Size Mode=::NewEnumerator1
    Uniform Sprite Size Max=15
    Uniform Sprite Size Min=7
    UsePositionOffset=true
    Write Material Random=true
   )
   &CameraOffset(Camera Offset Amount=7)
  ]
  particleUpdateScript [
   &ParticleState()
   &SocketLocation(
    SamplePosOffset=true
    Sampled Position Offset=&RandomRangeVector(Minimum=<-1 -1.8 -1.8>, Maximum=<1 1.8 1.8>)
    Socket Index=&RandomRangeInt(Minimum=0, Maximum=11)
    Socket Reader=$System.Socket Reader
    Socket Sampling Mode=::NewEnumerator3
    VelocitySpeedLimit=false
   )
   &ScaleSpriteSize(Scale Sprite Size Mode=::NewEnumerator2) {module=ScaleSpriteSize001}
   &SpriteRotationRate(Rotation Rate=&RandomRangeFloat(Minimum=-50, Maximum=50))
   &ScaleSpriteSize(
    Uniform Curve Scale=1.8
    Uniform Curve Sprite Scale=DataInterface() {propertyValues {Curve {keys [
        {arriveTangent=-0.6272, interpMode=RCIM_Cubic, time=0, value=0}
        {arriveTangent=-0.2398, interpMode=RCIM_Cubic, leaveTangent=-0.2398, tangentMode=RCTM_User, time=1, value=1}
       ]}}}
   )
   &Drag(Drag=2, Use Linear Drag=true, Use Rotational Drag=true)
   &SolveForcesAndVelocity(Clamp Velocity=false, Limit Acceleration=false)
   &ScaleColor(
    Scale Alpha=&FloatFromCurve(
     FloatCurve=DataInterface() {propertyValues {Curve {keys [
         <at 0 0 RCIM_Cubic>
         {arriveTangent=-0.0473, interpMode=RCIM_Cubic, leaveTangent=-0.0473, tangentMode=RCTM_User, time=0.4103, value=1}
         <at 1 0 RCIM_Cubic>
        ]}}}
     Scale Curve=&FloatFromCurve(
      FloatCurve=DataInterface() {propertyValues {Curve {keys [<at 0.1013 1>,<at 1 5>]}}}
      CurveIndex=$System.ReactivityIntensity
      Scale Curve=0.35
     )
    )
    Scale RGB=&MakeVectorFromLinearColorRGB(Color=&MultiplyLinearColor(
      Linear Color=&ColorFromCurve(ColorCurve=DataInterface() {propertyValues {
         AlphaCurve {keys []}
         BlueCurve {keys [<at 0.0717 1>, {arriveTangent=-1.2247, interpMode=RCIM_Cubic, leaveTangent=-1.2247, time=0.4137, value=0.5}, <at 0.8543 0.0417>]}
         GreenCurve {keys [<at 0.0717 1>, {arriveTangent=-1.2247, interpMode=RCIM_Cubic, leaveTangent=-1.2247, time=0.4137, value=0.5}, <at 0.8543 0.0417>]}
         RedCurve {keys [<at 0.0717 1>, {arriveTangent=-1.2247, interpMode=RCIM_Cubic, leaveTangent=-1.2247, time=0.4137, value=0.5}, <at 0.8543 0.0417>]}
        }})
      Scale Factor=&MakeVector4_FromLinearColor(LinearColor=&Multiply_LinearColorByFloat(
        Linear Color=&MergeHelper_ColorSwitch(
         0=<rgba 1.5 1.5 1.5 1>
         1=<rgba 3 0 0 1>
         2=<rgba 1.5652 0 3 1>
         3=<rgba 3 1.2522 0 1>
         Index=&MakeIntFromFloat(Float=$User.Variant, Conversion Mode=::NewEnumerator1)
         Num=::NewEnumerator2
        )
        Float=0.44
       ))
     ))
    ScaleA=true
    ScaleRGB=true
   )
   &ScaleColor(
    Scale Alpha=&FNTimeOfDayScalar(Night Scale=0.6, Morning Scale=0.7, Evening Scale=0.7)
    Scale RGB=&MakeVectorFromLinearColorRGB(Color=&FNTimeOfDayColor(Night Color=<rgba 1 0.9987 0.8229 1>, Morning Color=<rgba 1 1 1 1>, Day Color=<rgba 1 1 1 1>, Evening Color=<rgba 1 1 1 1>))
    ScaleA=true
    ScaleRGB=true
   ) {module=ScaleColor001}
   &FN_Sprite_Attributes(
    Color=&Multiply_LinearColorByFloat(Linear Color=$Particles.Color, Float=0.304)
    Create Color Attribute=true
    Create SpriteSize Attribute=true
    SpriteSize=&Multiply_Vector2D_ByFloat(Float=1.4, Vector2D=$Particles.SpriteSize)
    Write to Module Namespace=true
   ) {module=FN_Sprite_Attributes001}
  ]
  rendererClasses [&NiagaraSpriteRendererProperties]
  renderers [
   {rendererClass=&NiagaraSpriteRendererProperties}
   {rendererClass=&NiagaraSpriteRendererProperties, rendererIndex=1}
  ]
 }
 SignFire {
  emitterUpdateScript [
   &EmitterState(
    Inactive Response=::NewEnumerator1
    Life Cycle Mode=::NewEnumerator1
    Loop Behavior=::NewEnumerator1
    Loop Duration Mode=::NewEnumerator1
    UseLoopDelay=false
   )
   &SpawnBurst_Instantaneous(Spawn Count=6, Use Spawn Probability=false, UseLoopCountLimit=false)
  ]
  particleSpawnScript [
   &FN_SetSpawnGroup(Number of Spawn Groups=6)
   &InitializeParticle(
    Mesh Scale=&Multiply_VectorByFloat(
     Vector=&MergeHelper_VectorSwitch(
      0=<0.15 0.31 0.054>
      1=<0.15 0.41 0.092>
      2=<0.15 0.32 0.112>
      3=<0.15 0.37 0.14>
      4=<0.15 0.31 0.1>
      5=<0.15 0.23 0.07>
      Index=&SpawnGroup()
      Num=::NewEnumerator4
     )
     Float=1
    )
    Mesh Scale Mode=::NewEnumerator4
    UsePositionOffset=false
    Write Material Random=true
   )
   &DynamicMaterialParameters(
    Index 0 Param 1=&RandomRangeFloat(Minimum=0, Maximum=1)
    Index 0 Param 2=&RandomRangeFloat(Minimum=0, Maximum=10)
    Index 0 Param 3=30
    Index 0 Param 4=1
    Module.Param0WriteEnabled=true
    Module.Param1WriteEnabled=true
    Module.Param2WriteEnabled=false
    Module.Param3WriteEnabled=true
   )
   &InitialMeshOrientation(
    Mesh Orientation Mode=::NewEnumerator2
    Rotation=&MergeHelper_VectorSwitch(
     0=<-0.218 0 0>
     1=<-0.234 0 0>
     2=<-0.25 0 0>
     3=<-0.282 0 0>
     4=<-0.25 0 0>
     5=<-0.25 0 0>
     Index=&SpawnGroup()
     Num=::NewEnumerator4
    )
    Use Rotation Vector=true
   )
   &CameraOffset(Camera Offset Amount=5)
  ]
  particleUpdateScript [
   &ParticleState(Kill Particles When Lifetime Has Elapsed=false, Let Infinitely Lived Particles Die When Emitter Deactivates=true)
   &ScaleMeshSize(Scale Factor=&VectorFromFloat(Value=&FloatFromCurve(
      FloatCurve=DataInterface() {propertyValues {Curve {keys [<at 0 0>,<at 0.0957 1>,<at 0.8149 1>,<at 1 1>]}}}
      CurveIndex=$System.ReactivityIntensity
     )))
   &SocketLocation(
    SamplePosOffset=true
    Sampled Position Offset=&MergeHelper_VectorSwitch(
     0=<0 -2.7247 0.1689>
     1=<0 -3.8535 0.52>
     2=<0 -2.5002 0.06>
     3=<0 -1.8798 2.4244>
     4=<0 -3.5932 -0.514>
     5=<0 3.0725 -6.0555>
     Index=&SpawnGroup()
     Num=::NewEnumerator4
    )
    Socket Index=&MergeHelper_IntSwitch(0=0, 1=1, 2=3, 3=4, 4=8, 5=8, Index=&SpawnGroup(), Num=::NewEnumerator4)
    Socket Reader=$System.Socket Reader
    Socket Sampling Mode=::NewEnumerator3
    VelocitySpeedLimit=false
   )
   &InterpolateOverTime(Rate Of Change=7, Target Vector Value:NiagaraPosition=$Particles.Position, Variable Type=::NewEnumerator2)
   /SetVariables(Particles.Interpolate Dynamic=&MakeFloat4FromVectorAndFloat("Vector (RGB)"=$Output.InterpolateOverTime.Moving Average, "Float (Alpha)"=0.25)) {isSetParameters=true}
   &SolveForcesAndVelocity(Clamp Velocity=false, Limit Acceleration=false)
   &ScaleColor(
    Color Scale:LinearColor=&Multiply_LinearColorByFloat(
     Linear Color=&MergeHelper_ColorSwitch(
      0=<rgba 1.5 1.5 1.5 1>
      1=<rgba 3 0 0 1>
      2=<rgba 1.5652 0 3 1>
      3=<rgba 3 1.2522 0 1>
      Index=&MakeIntFromFloat(Float=$User.Variant, Conversion Mode=::NewEnumerator1)
      Num=::NewEnumerator2
     )
     Float=0.296
    )
    Scale Mode=3
   ) {module=ScaleColor001}
   &ScaleColor(
    Scale Alpha=&Multiply_Float(A=1, B=$System.ReactivityIntensity)
    Scale RGB=&Lerp_Vector(
     A=&FN_Select_Vector_from_Time_Of_Day(Morning=<1 1 1>, Day=<1 1 1>, Evening=<1 1 1>, Night=<0.25 0.25 0.25>)
     B=&FN_Select_Vector_from_Time_Of_Day(Morning=<1 1 1>, Day=<1 1 1>, Evening=<1 1 1>, Night=<0.75 0.75 0.75>)
     Alpha=$User.Variant
    )
    ScaleA=true
    ScaleRGB=true
   )
   &DynamicMaterialParameters(
    Index 0 Param 1=&RandomRangeFloat(Minimum=0, Maximum=1)
    Index 0 Param 2=&RandomRangeFloat(Minimum=0, Maximum=1)
    Index 0 Param 3=0.025
    Index 0 Param 4=1
    Index 3 Param 1=&RandomRangeFloat(Minimum=0.5, Maximum=1)
    Index 3 Param 2=&RandomRangeFloat(Minimum=0, Maximum=1)
    Index 3 Param 3=&RandomRangeFloat(Minimum=0, Maximum=0)
    Module.Index0Param0Write=true
    Module.Index0Param1Write=true
    Module.Index0Param2Write=true
    Module.Index0Param3Write=false
    Module.Index3Param0Write=false
    Module.Index3Param1Write=true
    Module.Index3Param2Write=true
    Module.Index3Param3Write=false
    Write Parameter Index 3=true
   ) {module=DynamicMaterialParameters002}
  ]
  rendererClasses [&NiagaraMeshRendererProperties]
  renderers [{rendererClass=&NiagaraMeshRendererProperties, rendererIndex=0}]
 }
 SpriteGlows {
  emitterUpdateScript [
   &EmitterState(
    Inactive Response=::NewEnumerator1
    Life Cycle Mode=::NewEnumerator1
    Loop Behavior=::NewEnumerator1
    Loop Duration Mode=::NewEnumerator1
    UseLoopDelay=false
   )
   &SpawnBurst_Instantaneous(Spawn Count=3, Use Spawn Probability=false, UseLoopCountLimit=false)
  ]
  particleSpawnScript [
   &FN_SetSpawnGroup(Number of Spawn Groups=3)
   &InitializeParticle(
    Sprite Rotation Mode=::NewEnumerator0
    Sprite Size=&MergeHelper_Vector2Switch(0=<12 20.7176>, 1=<12 27.2707>, 2=<9 47.7766>, 3=<9 29>, 4=?, Index=&SpawnGroup(), Num=::NewEnumerator2)
    Sprite Size Mode=::NewEnumerator4
    UsePositionOffset=false
    Write Material Random=true
   )
   &CameraOffset(Camera Offset Amount=4)
  ]
  particleUpdateScript [
   &ParticleState(Kill Particles When Lifetime Has Elapsed=false, Let Infinitely Lived Particles Die When Emitter Deactivates=true)
   &ScaleSpriteSize(
    Scale Sprite Size Mode=::NewEnumerator0
    Uniform Scale Factor=&FloatFromCurve(
     FloatCurve=DataInterface() {propertyValues {Curve {keys [<at 0 0>,<at 0.0957 1>,<at 0.8149 1>,<at 1 1>]}}}
     CurveIndex=$System.ReactivityIntensity
    )
   )
   &SocketLocation(
    SamplePosOffset=true
    Sampled Position Offset=&MergeHelper_VectorSwitch(
     0=<0 10.8504 55.8298>
     1=<0 -12.0706 54.4159>
     2=<0 -18.6266 41.5025>
     3=<0 -23.7102 -6.2603>
     Index=&SpawnGroup()
     Num=::NewEnumerator2
    )
    Socket Index=12
    Socket Reader=$System.Socket Reader
    Socket Sampling Mode=::NewEnumerator3
    VelocitySpeedLimit=false
   )
   &ScaleColor(
    Scale Alpha=&Multiply_Float(
     A=&FloatFromCurve(
      FloatCurve=DataInterface() {propertyValues {Curve {
         defaultValue=0
         keys [
          {arriveTangent=4.1785, interpMode=RCIM_Cubic, leaveTangent=4.1785, tangentMode=RCTM_User, time=0.1535, value=0}
          {arriveTangent=1.1179, interpMode=RCIM_Cubic, leaveTangent=1.1179, tangentMode=RCTM_User, time=0.362, value=0.6228}
          {arriveTangent=0.0224, interpMode=RCIM_Cubic, leaveTangent=0.0224, tangentMode=RCTM_User, time=1, value=1}
         ]
        }}}
      Scale Curve=$System.ReactivityIntensity
     )
     B=&FN_FrontEndFloatModifier(Gameplay=1, Front End=0.5)
    )
    Scale RGB=&MakeVectorFromLinearColorRGB(Color=&Multiply_LinearColorByFloat(
      Linear Color=&MergeHelper_ColorSwitch(
       0=<rgba 1.5 1.5 1.5 1>
       1=<rgba 3 0 0 1>
       2=<rgba 1.5652 0 3 1>
       3=<rgba 3 1.2522 0 1>
       Index=&MakeIntFromFloat(Float=$User.Variant, Conversion Mode=::NewEnumerator1)
       Num=::NewEnumerator2
      )
      Float=0.424
     ))
    ScaleA=true
    ScaleRGB=true
   )
   &SpriteFacingAndAlignment(
    Sprite Alignment=&MergeHelper_VectorSwitch(
     0=&RotateVectorByQuaternion(Vector To Rotate=<0 1 -0.4082>, Quaternion=$Output.SocketLocation.SampledRotation)
     1=&RotateVectorByQuaternion(Vector To Rotate=<0 1 -0.1033>, Quaternion=$Output.SocketLocation.SampledRotation)
     2=&RotateVectorByQuaternion(Vector To Rotate=<0 -0.096 1>, Quaternion=$Output.SocketLocation.SampledRotation)
     3=&RotateVectorByQuaternion(Vector To Rotate=<0 -0.424 1>, Quaternion=$Output.SocketLocation.SampledRotation)
     Index=&SpawnGroup()
     Num=::NewEnumerator2
    )
    WriteAlignment=true
    WriteFacing=false
   )
  ]
  rendererClasses [&NiagaraSpriteRendererProperties]
  renderers [{rendererClass=&NiagaraSpriteRendererProperties}]
 }
 TextureGlow {
  emitterUpdateScript [
   &EmitterState(
    Inactive Response=::NewEnumerator1
    Life Cycle Mode=::NewEnumerator1
    Loop Behavior=::NewEnumerator1
    Loop Duration Mode=::NewEnumerator1
    UseLoopDelay=false
   )
   &SpawnBurst_Instantaneous(Spawn Count=2, Use Spawn Probability=false, UseLoopCountLimit=false)
  ]
  particleSpawnScript [
   &FN_SetSpawnGroup(Number of Spawn Groups=2)
   &InitializeParticle(
    Lifetime=5
    Sprite Rotation Angle=-0.25
    Sprite Rotation Mode=::NewEnumerator1
    Sprite Size Mode=::NewEnumerator3
    Uniform Sprite Size=42.3784
    UsePositionOffset=false
    Write Material Random=true
   )
   &SpriteFacingAndAlignment(Sprite Alignment=<1 0 0>, Sprite Facing=<0 1 0>, WriteAlignment=true, WriteFacing=true)
   &CameraOffset(Camera Offset Amount=6)
   /SetVariables(Particles.VisibilityTag=&SpawnGroup()) {isSetParameters=true}
  ]
  particleUpdateScript [
   &ParticleState(
    Kill Particles When Lifetime Has Elapsed=false
    Let Infinitely Lived Particles Die When Emitter Deactivates=true
    Loop Particles Lifetime=true
   )
   &ScaleColor(
    Scale Alpha=&FloatFromCurve(
     FloatCurve=DataInterface() {propertyValues {Curve {
        defaultValue=0
        keys [
         {arriveTangent=4.1785, interpMode=RCIM_Cubic, leaveTangent=4.1785, tangentMode=RCTM_User, time=0.1535, value=0}
         {arriveTangent=1.1179, interpMode=RCIM_Cubic, leaveTangent=1.1179, tangentMode=RCTM_User, time=0.362, value=0.6228}
         {arriveTangent=0.0224, interpMode=RCIM_Cubic, leaveTangent=0.0224, tangentMode=RCTM_User, time=1, value=1}
        ]
       }}}
     CurveIndex=$Engine.System.Age
     Scale Curve=&Multiply_Float(A=&FN_FrontEndFloatModifier(Gameplay=0.15, Front End=0.15), B=$System.ReactivityIntensity)
    )
    Scale RGB=&MakeVectorFromLinearColorRGB(Color=&Multiply_LinearColorByFloat(
      Linear Color=&MergeHelper_ColorSwitch(
       0=<rgba 1.5 1.5 1.5 1>
       1=<rgba 3 0 0 1>
       2=<rgba 1.5652 0 3 1>
       3=<rgba 3 1.2522 0 1>
       Index=&MakeIntFromFloat(Float=$User.Variant, Conversion Mode=::NewEnumerator1)
       Num=::NewEnumerator2
      )
      Float=4.2
     ))
    ScaleA=true
    ScaleRGB=true
   )
   &SocketLocation(
    SamplePosOffset=true
    Sampled Position Offset=<0 -3.3353 44.4738>
    Socket Index=12
    Socket Reader=$System.Socket Reader
    Socket Sampling Mode=::NewEnumerator3
    VelocitySpeedLimit=false
   )
  ]
  rendererClasses [&NiagaraSpriteRendererProperties]
  renderers [{rendererClass=&NiagaraSpriteRendererProperties}]
 }
}
summary {
 systemName=NS_Pickaxe_CanyonVale_Idle
 emitters [
  {emitterName=TextureGlow, rendererClasses [&NiagaraSpriteRendererProperties]}
  {emitterName=SpriteGlows, rendererClasses [&NiagaraSpriteRendererProperties]}
  {emitterName=FirePiece, rendererClasses [&NiagaraSpriteRendererProperties]}
  {emitterName=SignFire, rendererClasses [&NiagaraMeshRendererProperties]}
 ]
 userVariables [
  User.bIsFrontEnd {type=NiagaraBool}
  User.bIsFrontEndPreview {type=NiagaraBool}
  User.DebugReactivity {type=NiagaraBool}
  User.ReactiveEmissive {type=LinearColor, description="", defaultValue {struct=LinearColor, value=<rgba 3 1.2522 0 1>}}
  User.ReactivityOn {type=NiagaraFloat, defaultValue=1, description=""}
  User.ReactivityTime {type=NiagaraFloat, defaultValue=0, description=""}
  User.Variant {type=NiagaraFloat, defaultValue=0, description=""}
 ]
}
