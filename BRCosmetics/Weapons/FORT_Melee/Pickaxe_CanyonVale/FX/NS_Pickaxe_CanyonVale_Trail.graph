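# Niagara swing-trail effect for the CanyonVale pickaxe, with a reactive look and color variants.
# Three ribbon anim-trail emitters (Fire, Hard_and_Soft_Trail, ScreenSoft) run between two bones or sockets, and a Sparks sprite emitter spawns at sockets with curl noise.
# System reactivity intensity and swing fade drive trail color, opacity and spark rate.
# User.Variant picks the palette, while time-of-day and front-end modifiers adjust tint and brightness.
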NiagaraSystem /BRCosmetics/Weapons/FORT_Melee/Pickaxe_CanyonVale/FX/NS_Pickaxe_CanyonVale_Trail
dependencies {
 usedDataInterfaces [&NiagaraDataInterfaceCurve, &NiagaraDataInterfaceColorCurve, &NiagaraDataInterfaceSocketReader]
 usedDynamicInputs [
  &MakeCustomFloatFromBool
  &Lerp_Float
  &Divide_Float
  &Subtract_Float
 ]
 usedModules [
  &self:SystemSpawnScript.SystemScriptSource.SystemScriptGraph.NiagaraNodeAssignment_1.SetVariables_#1
  &SystemState
  &self:SystemSpawnScript.SystemScriptSource.SystemScriptGraph.NiagaraNodeAssignment_0.SetVariables_#2
  &self:SystemSpawnScript.SystemScriptSource.SystemScriptGraph.NiagaraNodeAssignment_2.SetVariables_#3
 ]
}
emitters {
 Fire {
  emitterSpawnScript [&FN_RibbonIDOffsetUtility()]
  emitterUpdateScript [
   &EmitterState()
   &FN_SpawnAnimTrail()
  ]
  particleSpawnScript [
   &FN_InitializeAnimTrailBoneIndecies(First Bone or Socket Index=1, Second Bone or Socket Index=0)
   &FN_InitializeAnimTrailBoneIndecies(AnimTrail Selection=1, First Bone or Socket Index=1, Second Bone or Socket Index=0) {enabled=false, module=FN_InitializeAnimTrailBoneIndecies001}
   &FN_InitializeAnimTrail(AnimTrailWidth=0.97, Color=<rgba 3 3 3 1>, Lifetime=0.22)
   &AddVelocity(Velocity=<0 400 0>) {enabled=false}
  ]
  particleUpdateScript [
   &SolveForcesAndVelocity(Clamp Velocity=false, Limit Acceleration=false) {enabled=false}
   &FN_UpdateAnimTrail() {module=FN_UpdateAnimTrail001}
   &ScaleColor(
    Scale Alpha=&FloatFromCurve(
     FloatCurve=DataInterface() {propertyValues {Curve {keys [
         {arriveTangent=-0.4618, interpMode=RCIM_Cubic, leaveTangent=-0.4618, tangentMode=RCTM_User, time=0, value=1}
         {arriveTangent=-1.1299, interpMode=RCIM_Cubic, leaveTangent=-1.1299, tangentMode=RCTM_User, time=0.4269, value=0.2853}
         {arriveTangent=-0.0168, interpMode=RCIM_Cubic, tangentMode=RCTM_Break, time=1, value=0}
        ]}}}
     Scale Curve=&Multiply_Float(
      A=&FN_FrontEndFloatModifier(Gameplay=0.5, Front End=0.3)
      B=&FNTimeOfDayScalar(Night Scale=0.5, Morning Scale=0.6, Evening Scale=0.6)
     )
    )
    Scale RGB=&MakeVectorFromLinearColorRGB(Color=<rgba 1 1 1 1>)
    ScaleA=true
    ScaleRGB=true
   ) {enabled=false, module=ScaleColor002}
   &ScaleColor(
    Scale Alpha=&Multiply_Float(
     A=&FloatFromCurve(
      FloatCurve=DataInterface() {propertyValues {Curve {keys [
          <at 0 1.7371 RCIM_Cubic>
          {arriveTangent=-1.3606, interpMode=RCIM_Cubic, leaveTangent=-1.3606, time=0.2132, value=1.2596}
          {arriveTangent=-0.4634, interpMode=RCIM_Cubic, leaveTangent=-0.4634, tangentMode=RCTM_User, time=0.4416, value=1.1363}
          {arriveTangent=-0.0464, interpMode=RCIM_Cubic, tangentMode=RCTM_Break, time=1, value=0.5443}
         ]}}}
      Scale Curve=$System.ReactivityIntensity
     )
     B=&Multiply_Float(
      A=&FNTimeOfDayScalar(Night Scale=0.5, Morning Scale=0.6, Evening Scale=0.6)
      B=&Multiply_Float(A=0.404, B=$System.SwingGeneralFade)
     )
    )
    Scale RGB=&Multiply_VectorByFloat(
     Vector=&MakeVectorFromLinearColorRGB(Color=&MultiplyLinearColor(
       Linear Color=&MultiplyLinearColor(
        Linear Color=&ColorFromCurve(ColorCurve=DataInterface() {propertyValues {
           AlphaCurve {keys [<at 0 1 RCIM_Cubic>,<at 1 1 RCIM_Cubic>]}
           BlueCurve {keys [<at 0 8>, {arriveTangent=-7.9385, interpMode=RCIM_Cubic, leaveTangent=-7.9385, time=0.2763, value=3}, <at 0.6298 3 RCIM_Cubic>]}
           GreenCurve {keys [<at 0 8>, {arriveTangent=-7.9385, interpMode=RCIM_Cubic, leaveTangent=-7.9385, time=0.2763, value=3}, <at 0.6298 3 RCIM_Cubic>]}
           RedCurve {keys [<at 0 8>, {arriveTangent=-7.9385, interpMode=RCIM_Cubic, leaveTangent=-7.9385, time=0.2763, value=3}, <at 0.6298 3 RCIM_Cubic>]}
          }})
        Scale Factor=&MakeVector4_FromLinearColor(LinearColor=&MergeHelper_ColorSwitch(
          0=&ColorFromCurve(ColorCurve=DataInterface() {propertyValues {
             AlphaCurve {keys [<at 0 1 RCIM_Cubic>,<at 1 0 RCIM_Cubic>]}
             BlueCurve {keys [<at 0 1.3 RCIM_Cubic>,<at 1 1 RCIM_Cubic>]}
             GreenCurve {keys [<at 0 1.3 RCIM_Cubic>,<at 1 1 RCIM_Cubic>]}
             RedCurve {keys [<at 0 1.3 RCIM_Cubic>,<at 1 1 RCIM_Cubic>]}
            }})
          1=<rgba 3 0 0 1>
          2=<rgba 1.5652 0 3 1>
          3=&ColorFromCurve(ColorCurve=DataInterface() {propertyValues {
             AlphaCurve {keys [<at 0 1 RCIM_Cubic>,<at 1 0 RCIM_Cubic>]}
             BlueCurve {keys [<at 0 0 RCIM_Cubic>,<at 1 0 RCIM_Cubic>]}
             GreenCurve {keys [<at 0 1.2522 RCIM_Cubic>,<at 1 0.5755 RCIM_Cubic>]}
             RedCurve {keys [<at 0 3 RCIM_Cubic>,<at 1 3 RCIM_Cubic>]}
            }})
          Index=&MakeIntFromFloat(Float=$User.Variant, Conversion Mode=::NewEnumerator1)
          Num=::NewEnumerator2
         ))
       )
       Scale Factor=&MakeVector4_FromLinearColor(LinearColor=&FNTimeOfDayColor(
         Night Color=<rgba 0.7691 0.9321 1 1>
         Morning Color=<rgba 1 0.776 0.776 1>
         Day Color=<rgba 1 0.776 0.776 1>
         Evening Color=<rgba 1 0.776 0.776 1>
        ))
      ))
     Float=&FloatFromCurve(FloatCurve=DataInterface() {propertyValues {Curve {keys [<at 0 4 RCIM_Cubic>,<at 0.2335 1>]}}})
    )
    ScaleA=true
    ScaleRGB=true
   )
   &ScaleRibbonWidth(Ribbon Width Scale=&FloatFromCurve(FloatCurve=DataInterface() {propertyValues {Curve {keys [<at 0 1>,<at 1 0.9136>]}}}))
   &DynamicMaterialParameters(
    Index 0 Param 1=&FloatFromCurve(FloatCurve=DataInterface() {propertyValues {Curve {keys [
         <at 0.2889 -2.0689 RCIM_Cubic>
         {arriveTangent=-14.6907, interpMode=RCIM_Cubic, leaveTangent=-14.6907, tangentMode=RCTM_User, time=0.8, value=-7}
        ]}}})
    Module.Index0Param0Write=true
    Module.Index0Param1Write=true
    Module.Index0Param2Write=true
    Module.Index0Param3Write=true
   )
   /SetVariables(Particles.DarkWidth=80) {isSetParameters=true}
   /SetVariables(Particles.DarkColor:LinearColor=&MakeLinearColorFromVectorAndFloat(
     "Vector (RGB)"=&MakeVectorFromLinearColorRGB(Color=&MultiplyLinearColor(
       Linear Color=&ColorFromCurve(ColorCurve=DataInterface() {propertyValues {
          AlphaCurve {keys []}
          BlueCurve {keys [<at 0 0.0469 RCIM_Cubic>,<at 0.4118 0.0104 RCIM_Cubic>]}
          GreenCurve {keys [<at 0 0.0469 RCIM_Cubic>,<at 0.4118 0.0104 RCIM_Cubic>]}
          RedCurve {keys [<at 0 0.0469 RCIM_Cubic>,<at 0.4118 0.0104 RCIM_Cubic>]}
         }})
       Scale Factor=&MakeVector4_FromLinearColor(LinearColor=&MergeHelper_ColorSwitch(
         0=<rgba 2 2 2 1>
         1=<rgba 3 0 0 1>
         2=<rgba 1.5652 0 3 1>
         3=<rgba 3 1.2522 0 1>
         Index=&MakeIntFromFloat(Float=$User.Variant, Conversion Mode=::NewEnumerator1)
         Num=::NewEnumerator2
        ))
      ))
     "Float (Alpha)"=&FloatFromCurve(
      FloatCurve=DataInterface() {propertyValues {Curve {keys [
          {arriveTangent=-0.0791, interpMode=RCIM_Cubic, leaveTangent=-0.0791, tangentMode=RCTM_User, time=0, value=1}
          {arriveTangent=-0.4815, interpMode=RCIM_Cubic, leaveTangent=-0.4815, tangentMode=RCTM_User, time=0.4337, value=0.7559}
          <at 0.8 0 RCIM_Cubic>
         ]}}}
      Scale Curve=&Multiply_Float(A=4, B=&Multiply_Float(A=$System.ReactivityIntensity, B=$System.SwingGeneralFade))
     )
    )) {isSetParameters=true}
  ]
  rendererClasses [&NiagaraRibbonRendererProperties]
  renderers [
   {rendererClass=&NiagaraRibbonRendererProperties}
   {rendererClass=&NiagaraRibbonRendererProperties, rendererIndex=1}
   {rendererClass=&NiagaraRibbonRendererProperties, rendererIndex=2}
  ]
 }
 Hard_and_Soft_Trail {
  emitterSpawnScript [&FN_RibbonIDOffsetUtility()]
  emitterUpdateScript [
   &EmitterState()
   &FN_SpawnAnimTrail(AnimTrailWidthMode=::FromSecond)
  ]
  particleSpawnScript [
   &FN_InitializeAnimTrailBoneIndecies(First Bone or Socket Index=1, Second Bone or Socket Index=0)
   &FN_InitializeAnimTrail(AnimTrailWidth=&MaskFloatBySpawnGroup(Masked Float=1, Passthrough Float=0.7), Lifetime=0.2)
  ]
  particleUpdateScript [
   &SolveForcesAndVelocity(Clamp Velocity=false, Limit Acceleration=false)
   &FN_UpdateAnimTrail() {module=FN_UpdateAnimTrail001}
   &ScaleColor(
    Color Scale:LinearColor=&LerpLinearColors(StartColor=<rgba 0.1979 0.1979 0.1979 1>, EndColor=<rgba 1 0.5973 0.0024 0>, LerpFactor=$System.ReactivityIntensity)
    Scale Mode=3
   ) {module=ScaleColor001}
   &ScaleColor(
    Scale Alpha=&FloatFromCurve(
     FloatCurve=DataInterface() {propertyValues {Curve {keys [{interpMode=RCIM_Cubic, tangentMode=RCTM_User, time=0, value=1}, {interpMode=RCIM_Cubic, tangentMode=RCTM_User, time=1, value=0}]}}}
     Scale Curve=$System.SwingGeneralFade
    )
    Scale RGB=&MakeVectorFromLinearColorRGB(Color=&FNTimeOfDayColor(Night Color=<rgba 0.32 0.4976 0.8 1>, Morning Color=<rgba 0.6 0.6262 0.8 1>, Day Color=<rgba 1 1 1 1>, Evening Color=<rgba 0.6 0.6 0.6 1>))
    ScaleA=true
    ScaleRGB=true
   )
  ]
  rendererClasses [&NiagaraRibbonRendererProperties]
  renderers [{rendererClass=&NiagaraRibbonRendererProperties}]
 }
 ScreenSoft {
  emitterSpawnScript [&FN_RibbonIDOffsetUtility()]
  emitterUpdateScript [
   &EmitterState()
   &FN_SpawnAnimTrail(AnimTrailWidthMode=::FromSecond)
  ]
  particleSpawnScript [
   &FN_InitializeAnimTrailBoneIndecies(First Bone or Socket Index=1, Second Bone or Socket Index=0)
   &FN_InitializeAnimTrail(AnimTrailWidth=0.6, Color=<rgba 1 1 1 0.8>, Lifetime=0.2)
  ]
  particleUpdateScript [
   &FN_UpdateAnimTrail() {module=FN_UpdateAnimTrail001}
   &ScaleRibbonWidth(Ribbon Width Scale=&FloatFromCurve(
     FloatCurve=DataInterface() {propertyValues {Curve {keys [{arriveTangent=-0.0514, time=0, value=1}, {arriveTangent=-1.0387, leaveTangent=-1.0387, time=1, value=0.6}]}}}
     Scale Curve=&FN_FrontEndFloatModifier(Gameplay=1, Front End=0.8)
    )) {module=ScaleRibbonWidth001}
   &ScaleColor(
    Scale Alpha=&FloatFromCurve(
     FloatCurve=DataInterface() {propertyValues {Curve {keys [
         {arriveTangent=-1.6719, interpMode=RCIM_Cubic, leaveTangent=-1.6719, tangentMode=RCTM_User, time=0, value=1}
         {arriveTangent=0.0495, interpMode=RCIM_Cubic, leaveTangent=0.0495, tangentMode=RCTM_User, time=1, value=0}
        ]}}}
     Scale Curve=0.5
    )
    Scale RGB=&MakeVectorFromLinearColorRGB(Color=&Multiply_LinearColorByFloat(
      Linear Color=&MergeHelper_ColorSwitch(
       0=<rgba 1.5 1.5 1.5 1>
       1=<rgba 3 0 0 1>
       2=<rgba 1.5652 0 3 1>
       3=<rgba 3 1.2522 0 1>
       Index=&MakeIntFromFloat(Float=$User.Variant, Conversion Mode=::NewEnumerator1)
       Num=::NewEnumerator2
      )
      Float=1.3
     ))
    ScaleA=true
    ScaleRGB=true
   ) {module=ScaleColor001}
   /SetVariables(
    Particles.BaseColor:LinearColor=&MultiplyLinearColor(
     Linear Color=$Particles.Color
     Scale Factor=&MakeFloat4FromVectorAndFloat(
      "Vector (RGB)"=&MakeVectorFromLinearColorRGB(Color=&MultiplyLinearColor(Linear Color=<rgba 1 1 1 1>, Scale Factor=<1 1 1 1>))
      "Float (Alpha)"=&FloatFromCurve(
       FloatCurve=DataInterface() {propertyValues {Curve {defaultValue=0, keys [{tangentMode=RCTM_User, time=0, value=0}, {tangentMode=RCTM_User, time=1, value=1}]}}}
       CurveIndex=$System.ReactivityIntensity
      )
     )
    )
    Particles.ReactiveColor:LinearColor=&MultiplyLinearColor(
     Linear Color=$Particles.Color
     Scale Factor=&MakeFloat4FromVectorAndFloat(
      "Vector (RGB)"=<2 2 2>
      "Float (Alpha)"=&FloatFromCurve(
       FloatCurve=DataInterface() {propertyValues {Curve {defaultValue=1, keys [{tangentMode=RCTM_User, time=0, value=0}, {tangentMode=RCTM_User, time=1, value=1}]}}}
       CurveIndex=$System.ReactivityIntensity
       Scale Curve=$System.SwingGeneralFade
      )
     )
    )
   ) {isSetParameters=true}
  ]
  rendererClasses [&NiagaraRibbonRendererProperties]
  renderers [{rendererClass=&NiagaraRibbonRendererProperties}]
 }
 Sparks {
  emitterUpdateScript [
   &EmitterState(Life Cycle Mode=::NewEnumerator1, Loop Behavior=::NewEnumerator1, Loop Delay=0.01, Loop Duration=0.23, UseLoopDelay=true)
   &SpawnRate(
    Spawn Probability=0.7
    SpawnRate=&Multiply_Float(A=70, B=$System.ReactivityIntensity)
    Use Spawn Probability=true
   )
  ]
  particleSpawnScript [
   &InitializeParticle(
    Lifetime Max=0.55
    Lifetime Min=0.35
    Lifetime Mode=::NewEnumerator1
    Mesh Scale Mode=::NewEnumerator1
    Mesh Uniform Scale Max=1.5
    Mesh Uniform Scale Min=0.5
    Sprite Size Max=<2 6>
    Sprite Size Min=<1 2>
    Sprite Size Mode=::NewEnumerator2
    UsePositionOffset=true
    Write Material Random=true
   )
   &SocketLocation(SamplePosOffset=false, Socket Reader=$System.Socket Reader, Socket Sampling Mode=::NewEnumerator3, VelocitySpeedLimit=false)
   &ShapeLocation(
    MaskSpawnGroup=false
    Offset Mode=::NewEnumerator0
    Override Seed=false
    Rotation Axis=<0 1 0>
    Rotation Mode=::NewEnumerator1
    Sphere Radius=30
    UseOwnerScale=true
   )
   &AddVelocity(Velocity=&RandomRangeVector(Minimum=<0 0 80>, Maximum=<0 0 100>))
  ]
  particleUpdateScript [
   &ParticleState()
   &ScaleSpriteSize(
    Uniform Curve Scale=0.8
    Uniform Curve Sprite Scale=DataInterface() {propertyValues {Curve {keys [<at 0 0 RCIM_Cubic>,<at 0.2 1 RCIM_Cubic>,<at 1 0 RCIM_Cubic>]}}}
   )
   /SetVariables(Particles.Glow03_Size=&Multiply_Vector2D_ByFloat(
     Float=&FloatFromCurve(
      FloatCurve=DataInterface() {propertyValues {Curve {keys [
          <at 0 0 RCIM_Cubic>
          <at 0.2 1.3 RCIM_Cubic>
          {arriveTangent=-0.1423, interpMode=RCIM_Cubic, leaveTangent=-0.1423, tangentMode=RCTM_User, time=1, value=0}
         ]}}}
      Scale Curve=8
     )
     Vector2D=<0.7 1.5>
    )) {isSetParameters=true}
   &ScaleColor(
    Scale Alpha=&Multiply_Float(
     A=1
     B=&FloatFromCurve(FloatCurve=DataInterface() {propertyValues {Curve {keys [
          {interpMode=RCIM_Cubic, tangentMode=RCTM_User, time=0.1, value=0}
          {arriveTangent=-1.3622, interpMode=RCIM_Cubic, leaveTangent=-1.3622, tangentMode=RCTM_User, time=0.2, value=1}
          {tangentMode=RCTM_User, time=0.8, value=0}
         ]}}})
    )
    Scale RGB=&MakeVectorFromLinearColorRGB(Color=&Multiply_LinearColorByFloat(
      Linear Color=&MergeHelper_ColorSwitch(
       0=<rgba 1.5 1.5 1.5 1>
       1=<rgba 3 0 0 1>
       2=<rgba 2.0869 0 4 1>
       3=<rgba 3 1.2522 0 1>
       Index=&MakeIntFromFloat(Float=$User.Variant, Conversion Mode=::NewEnumerator1)
       Num=::NewEnumerator2
      )
      Float=1.2
     ))
    ScaleA=true
    ScaleRGB=true
   )
   /SetVariables(Particles.Glow03_Color=&MakeLinearColorFromVectorAndFloat(
     "Vector (RGB)"=&MakeVectorFromLinearColorRGB(Color=&MultiplyLinearColor(
       Linear Color=&MergeHelper_ColorSwitch(
        0=<rgba 1.5 1.5 1.5 1>
        1=<rgba 3 0 0 1>
        2=<rgba 1.5652 0 3 1>
        3=<rgba 3 1.2522 0 1>
        Index=&MakeIntFromFloat(Float=$User.Variant, Conversion Mode=::NewEnumerator1)
        Num=::NewEnumerator2
       )
       Scale Factor=&MakeVector4_FromFloat(Float=0.5)
      ))
     "Float (Alpha)"=&Multiply_Float(A=0.2, B=1)
    )) {isSetParameters=true}
   &GravityForce(Gravity=&RandomRangeVector(Minimum=<0 0 50>, Maximum=<0 0 -100>))
   &CurlNoiseForce(
    Bias Noise=false
    Noise Strength=&RandomRangeFloat(Minimum=200, Maximum=400)
    Pan Noise=true
    Pan Noise Field=<0 0 1>
   )
   &Drag(Drag=3, Use Linear Drag=true, Use Rotational Drag=true)
   &SolveForcesAndVelocity(Clamp Velocity=false, Limit Acceleration=false)
  ]
  rendererClasses [&NiagaraSpriteRendererProperties]
  renderers [
   {rendererClass=&NiagaraSpriteRendererProperties}
   {rendererClass=&NiagaraSpriteRendererProperties, rendererIndex=1}
  ]
 }
}
summary {
 systemName=NS_Pickaxe_CanyonVale_Trail
 emitters [
  {emitterName=Hard_and_Soft_Trail, rendererClasses [&NiagaraRibbonRendererProperties]}
  {emitterName=ScreenSoft, rendererClasses [&NiagaraRibbonRendererProperties]}
  {emitterName=Fire, rendererClasses [&NiagaraRibbonRendererProperties]}
  {emitterName=Sparks, rendererClasses [&NiagaraSpriteRendererProperties]}
 ]
 userVariables [
  User.bIsFrontEnd {type=NiagaraBool}
  User.bIsFrontEndPreview {type=NiagaraBool}
  User.DebugReactivity {type=NiagaraBool}
  User.ReactiveEmissive {type=LinearColor, description="", defaultValue {struct=LinearColor, value=<rgba 3 1.2522 0 1>}}
  User.ReactivityOn {type=NiagaraFloat, defaultValue=1, description=""}
  User.ReactivityTime {type=NiagaraFloat, defaultValue=0, description=""}
  User.RibbonIDOffset {type=NiagaraInt32}
  User.Variant {type=NiagaraFloat, defaultValue=0, description=""}
 ]
}
