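# Material function for the CanyonVale pickaxe that adds a timestamp-driven reactive glow and metal recolor to incoming material attributes.
# Time since the Timestamp input, scaled by a duration and shaped by a power, forms the reactivity curve.
# Masks tint BaseColor and shape an emissive color, which ToggleGlow blends into BaseColor and EmissiveColor.
# Panning fire noise, handle mask, Fresnel rims and an isDefaultPickaxe orange-tint variant form supporting branches.
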MaterialFunction /BRCosmetics/Weapons/FORT_Melee/Pickaxe_CanyonVale/Materials/MF_CanyonValePickaxe
graph {nodes [
  n1 = GetMaterialAttributes(None=FunctionInput() {inputType=FunctionInput_MaterialAttributes}) {attributeGetTypes [#1, #2]}
  n2 = ScalarParameter() {defaultValue=1, group=Reactivity, parameterName=ToggleGlow}
  n3 = ComponentMask(None=NamedRerouteUsage()) {r=true}
  n4 = LinearInterpolate(B=NamedRerouteUsage(), Alpha=@n3)
  n5 = Subtract(A=@n4, B=NamedRerouteUsage())
  n6 = Constant() {r=0.75}
  n7 = LinearInterpolate(
   B=VectorParameter() {group=Reactivity, parameterName=ReactiveEmissive, defaultValue=<rgba 5 2.087 0 1>}.RGB
   Alpha=Saturate(None=Divide(A=@n5, B=@n6))
  ) {constAlpha=1, constB=0}
  n8 = VectorParameter() {group=Reactivity, parameterName=MetalColor, defaultValue=<rgba 0.0608 0 0.0364 1>}
  n9 = NamedRerouteDeclaration(None=LinearInterpolate(A=@n1.BaseColor, B=Multiply(A=@n8.RGB, B=@n1.BaseColor), Alpha=NamedRerouteUsage())) {name=BaseColor_IN, nodeColor=<rgba 0.0312 0 1 1>}
  n10 = Add(A=@n9, B=@n7)
  FunctionOutput(None=SetMaterialAttributes!(
    MaterialAttributes=@n1.MaterialAttributes
    BaseColor=LinearInterpolate(A=@n9, B=@n10, Alpha=@n2)
    EmissiveColor=LinearInterpolate(B=@n7, Alpha=@n2)
   ) {attributeSetTypes [#1, #2]})
  n11 = Fresnel(
   ExponentIn=ScalarParameter() {defaultValue=0.5, group=Reactivity, parameterName=ExponentIn}
   BaseReflectFractionIn=ScalarParameter() {defaultValue=0.04, group=Reactivity, parameterName=BaseReflectFractionIn}
  )
  n12 = Fresnel(
   ExponentIn=ScalarParameter() {defaultValue=0.8, group=Reactivity, parameterName=ExponentIn2}
   BaseReflectFractionIn=ScalarParameter() {defaultValue=0.005, group=Reactivity, parameterName=BaseReflectFractionIn2}
  )
  n13 = Saturate(None=Power(Base=OneMinus(None=@n12)) {constExponent=11})
  n14 = Multiply(A=Add(A=Multiply(A=@n11) {constB=0.3}, B=Multiply(A=@n13) {constB=0.5}), B=NamedRerouteUsage())
  n15 = Multiply(A=Constant() {r=0.05}, B=NamedRerouteUsage())
  n16 = Saturate(None=Multiply(A=NamedRerouteUsage(), B=Add(A=@n14, B=@n15)))
  LinearInterpolate(A=@n4, B=Saturate(None=Add(A=@n16, B=@n3)), Alpha=ScalarParameter() {parameterName=isDefaultPickaxe})
  NamedRerouteDeclaration(None=@n1.EmissiveColor) {name=EmissiveMask, nodeColor=<rgba 0.7344 0 1 1>}
  n17 = Subtract(A=Time(), B=FunctionInput() {inputName=Timestamp, inputType=FunctionInput_Scalar})
  n18 = ScalarParameter() {defaultValue=7, group=Reactivity, parameterName=ReactivityDuration}
  n19 = Abs(None=ConstantBiasScale(None=Divide(A=@n17, B=@n18)) {bias=0, scale=2})
  n20 = ScalarParameter() {defaultValue=4, group=Reactivity, parameterName=EmissiveFadePower}
  NamedRerouteDeclaration(None=Saturate(None=Power(Base=@n19, Exp=@n20) {constExponent=0.8})) {name=ReactivityCurve, nodeColor=<rgba 0 0.6719 1 1>}
  NamedRerouteDeclaration(None=TextureSample() {samplerType=SAMPLERTYPE_Grayscale, texture=/BRCosmetics/Weapons/FORT_Melee/Pickaxe_CanyonVale/Textures/T_CanyonVale_Pickaxe_FX1}.R) {name=HandleMask, nodeColor=<rgba 1 0.75 0 1>}
  n21 = FunctionInput(Preview=TextureSample() {samplerType=SAMPLERTYPE_Masks, texture=/BRCosmetics/Weapons/FORT_Melee/Pickaxe_CanyonVale/Textures/T_CanyonVale_Pickaxe_FX}.RGB) {inputName=TechArt Mask}
  NamedRerouteDeclaration(None=MaterialFunctionCall(In=ComponentMask(None=@n21) {b=true}, Contrast=Constant() {r=0.1}) {materialFunction=&CheapContrast}.Result) {name=MetallicRecolorMask, nodeColor=<rgba 0 1 0.7188 1>}
  n22 = TextureCoordinate()
  n23 = Multiply(A=Time()) {constB=0.2}
  NamedRerouteDeclaration(None=TextureSample(UVs=Add(A=@n22, B=@n23)) {texture=/BRCosmetics/Weapons/FORT_Melee/Pickaxe_CanyonVale/Textures/T_CanyonVale_FireNoise_02}.R) {name=FireTex, nodeColor=<rgba 0 1 0.7891 1>}
  n24 = NamedRerouteUsage()
  n25 = NamedRerouteUsage()
  n26 = Constant3Vector() {constant=<rgba 1 0.3468 0 1>}
  n27 = LinearInterpolate(A=@n25, B=Multiply(A=Multiply(A=@n25, B=@n26)) {constB=2}, Alpha=NamedRerouteUsage())
  n28 = ScalarParameter() {parameterName=isDefaultPickaxe}
  n29 = Add(A=NamedRerouteUsage()) {constB=0.4}
  NamedRerouteDeclaration(None=LinearInterpolate(A=LinearInterpolate(A=@n24, B=@n27, Alpha=@n28), B=NamedRerouteUsage(), Alpha=Saturate(None=@n29)) {constAlpha=0}) {name=OrangeTint, nodeColor=<rgba 1 0.4778 0 1>}
 ]}
