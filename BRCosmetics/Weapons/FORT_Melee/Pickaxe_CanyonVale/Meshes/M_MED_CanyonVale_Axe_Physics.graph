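# Physics asset for the CanyonVale pickaxe mesh, providing collision bodies and joints for a dangling chain on the axe.
# A larger capsule on dyn_chain_base anchors six small, identical capsules on dyn_chain_01 through dyn_chain_06.
# Constraints link each chain bone to the one before it, allowing wide swing while keeping twist tightly limited, so the links bend without spinning.
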PhysicsAsset /BRCosmetics/Weapons/FORT_Melee/Pickaxe_CanyonVale/Meshes/M_MED_CanyonVale_Axe_Physics
bodies {
 dyn_chain_01 [{shapeType=Capsule, length=1.5, radius=1.1, shapeName=dyn_chain_a_capsule, center=<0.6 0 0>, rotation=<rot 0 90 -90>}]
 dyn_chain_02 [{shapeType=Capsule, length=1.5, radius=1.1, shapeName=dyn_chain_a_capsule, center=<0.6 0 0>, rotation=<rot 0 90 -90>}]
 dyn_chain_03 [{shapeType=Capsule, length=1.5, radius=1.1, shapeName=dyn_chain_a_capsule, center=<0.6 0 0>, rotation=<rot 0 90 -90>}]
 dyn_chain_04 [{shapeType=Capsule, length=1.5, radius=1.1, shapeName=dyn_chain_a_capsule, center=<0.6 0 0>, rotation=<rot 0 90 -90>}]
 dyn_chain_05 [{shapeType=Capsule, length=1.5, radius=1.1, shapeName=dyn_chain_a_capsule, center=<0.6 0 0>, rotation=<rot 0 90 -90>}]
 dyn_chain_06 [{shapeType=Capsule, length=1.5, radius=1.1, shapeName=dyn_chain_a_capsule, center=<0.6 0 0>, rotation=<rot 0 90 -90>}]
 dyn_chain_base [{
   shapeType=Capsule
   length=10.3908
   radius=3.2117
   shapeName=dyn_chain_base_capsule
   center=<-10.0423 0 -0.0208>
   rotation=<rot 0 -90 -90.5833>
  }]
}
constraints [
 {bone1Name=dyn_chain_01, bone2Name=dyn_chain_base, swing1LimitDegrees=100, swing2LimitDegrees=90, twistLimitDegrees=4}
 {bone1Name=dyn_chain_02, bone2Name=dyn_chain_01, swing1LimitDegrees=100, swing2LimitDegrees=80, twistLimitDegrees=4}
 {bone1Name=dyn_chain_03, bone2Name=dyn_chain_02, swing1LimitDegrees=80, swing2LimitDegrees=85, twistLimitDegrees=4}
 {bone1Name=dyn_chain_04, bone2Name=dyn_chain_03, swing1LimitDegrees=80, swing2LimitDegrees=85, twistLimitDegrees=4}
 {bone1Name=dyn_chain_05, bone2Name=dyn_chain_04, swing1LimitDegrees=85, swing2LimitDegrees=85, twistLimitDegrees=4}
 {bone1Name=dyn_chain_06, bone2Name=dyn_chain_05, swing1LimitDegrees=90, swing2LimitDegrees=85, twistLimitDegrees=4}
]
