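# Animation Blueprint for the CanyonVale pickaxe that adds rigid body physics motion to the axe's bones.
# The cached input pose feeds two RigidBody nodes, and bIsInFrontEnd selects the base-bone-space simulation in the front end or the world-space one with fixed tuning otherwise.
# In the front end, the update event switches gravity and sim-space settings for the base-bone simulation according to bIsOnPawn.
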AnimBlueprint /BRCosmetics/Weapons/FORT_Melee/Pickaxe_CanyonVale/Meshes/M_MED_CanyonVale_Axe_AnimBP
target_skeleton=&/M_MED_CanyonVale_Axe_Skeleton
graphs [
 AnimGraph {nodes [
   n1 = RigidBody!(
    ComponentPose=Local To Component(LocalPose=Use cached pose 'presim'())
    Alpha
    OverrideWorldGravity:Vector=Get Gravity() -> (Gravity:Vector)
    ExternalForce
    SimSpaceSettings:Sim Space Settings Structure=Get SimSpaceSettings() -> (SimSpaceSettings:Sim Space Settings Structure)
   ) -> (Pose) {
    bOverrideWorldGravity=true
    overridePhysicsAsset=&/M_MED_CanyonVale_Axe_Physics
    simulationSpace=BaseBoneSpace
    baseBoneRef {boneName=b_AR_Root}
    overrideWorldGravity {z=-1015}
    simSpaceSettings {
     dampingAlpha=0.4
     maxAngularAcceleration=225
     maxAngularVelocity=45
     maxLinearAcceleration=1500
     maxLinearVelocity=250
     worldAlpha=1
     externalLinearDragV=<0.9 0.9 0.9>
    }
   }
   n2 = RigidBody!(ComponentPose=Local To Component(LocalPose=Use cached pose 'presim'()), Alpha, OverrideWorldGravity="0,0,-1300", ExternalForce) -> (Pose) {
    bClampLinearTranslationLimitToRefPose=true
    bOverrideWorldGravity=true
    overridePhysicsAsset=&M_MED_CanyonVale_Axe_Physics
    simulationSpace=WorldSpace
    alphaBoolBlend {blendOption=Linear}
    baseBoneRef {boneName=Melee_TwoHanded_Handle}
    simSpaceSettings {
     dampingAlpha=0
     maxAngularAcceleration=12
     maxAngularVelocity=3
     maxLinearAcceleration=1550
     maxLinearVelocity=600
     worldAlpha=0.7
     externalLinearDragV=<1 1 1>
    }
   }
   Output Pose(Result=Blend Poses by bool(BlendPose_0=Component To Local(ComponentPose=@n1), BlendPose_1=Component To Local(ComponentPose=@n2), bActiveValue=Get bIsInFrontEnd())) {layerGroup=None}
   AnimGraphNode_SubInput_0 = Input Pose()
   Save cached pose 'presim'(Pose=@AnimGraphNode_SubInput_0) {cacheName=presim}
  ]}
 EventGraph {nodes [
   Event BlueprintUpdateAnimation()
   > Branch(Condition=Get bIsInFrontEnd())
   > n1 = Branch(Condition=Get bIsOnPawn() -> (bIsOnPawn:bool))
   > Set SimSpaceSettings(SimSpaceSettings:Sim Space Settings Structure=Get SimSpaceSettings_2() -> (SimSpaceSettings_2:Sim Space Settings Structure)) -> (Output_Get:Sim Space Settings Structure)
   > Set Gravity(Gravity:Vector="0,0,-1980") -> (Output_Get:Vector)
   Set SimSpaceSettings(execute=@n1.else, SimSpaceSettings=Get SimSpaceSettings_0() -> (SimSpaceSettings_0:Sim Space Settings Structure))
   > Set Gravity(Gravity="0,0,-980")
   TryGetPawnOwner()
   TryGetPawnOwner()
  ]}
]
variables [
 Gravity {type=Vector}
 SimSpaceSettings {
  type=SimSpaceSettings
  default {
   dampingAlpha=1
   maxAngularAcceleration=10000
   maxAngularVelocity=10000
   maxLinearAcceleration=10000
   maxLinearVelocity=10000
   velocityScaleZ=1
   worldAlpha=0
   externalLinearDragV=<0 0 0>
  }
 }
 SimSpaceSettings_0 {
  type=SimSpaceSettings
  default {
   dampingAlpha=0
   maxAngularAcceleration=70
   maxAngularVelocity=7
   maxLinearAcceleration=800
   maxLinearVelocity=200
   velocityScaleZ=0.2
   worldAlpha=1
   externalLinearDragV=<0 0 0>
  }
 }
 SimSpaceSettings_1 {
  type=SimSpaceSettings
  default {
   dampingAlpha=0.45
   maxAngularAcceleration=20
   maxAngularVelocity=5
   maxLinearAcceleration=1800
   maxLinearVelocity=320
   velocityScaleZ=0.1
   worldAlpha=0.7936
   externalLinearDragV=<1 1 1>
  }
 }
 SimSpaceSettings_2 {
  type=SimSpaceSettings
  default {
   dampingAlpha=0.1
   maxAngularAcceleration=48
   maxAngularVelocity=4.8
   maxLinearAcceleration=785
   maxLinearVelocity=200
   velocityScaleZ=0.2
   worldAlpha=1
   externalLinearDragV=<0 0 0>
  }
 }
]
